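# Unlit, translucent, two-sided Niagara ribbon material that draws the RadioPaca glider trail as an LED spectrum bar graph.
# UVs are quantized into a cell grid, and time-scrolled noise is thresholded to form bars with faded masking and cell outlines.
# A color gradient along U adds noise-driven bright spots, then passes through an emissive adjustment function.
# Particle dynamic parameters control noise, bright spots and scroll speed; particle alpha scales opacity.
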Material /BRCosmetics/Gadgets/Assets/VinderTech_GliderChute/Glider_RadioPaca/FX/M_Glider_RadioPaca_Trail
graph {
 properties {
  bAutomaticallySetUsageInEditor=false
  bUsedWithNiagaraRibbons=true
  blendMode=BLEND_Translucent
  shadingModel=MSM_Unlit
  twoSided=true
  usedShadingModels=Unlit
 }
 nodes [
  n1 = ComponentMask(None=TextureCoordinate()) {r=true}
  n2 = ScalarParameter() {defaultValue=1, parameterName=Color Lerp Power, sliderMax=5}
  n3 = NamedRerouteUsage()
  n4 = Constant() {r=0.8}
  n5 = Multiply(
   A=LinearInterpolate(
    A=VectorParameter() {parameterName=Color Low, defaultValue=<rgba 0.0468 0.1037 0.1927 1>}.RGB
    B=VectorParameter() {parameterName=Color High, defaultValue=<rgba 0.15 0.493 3 1>}.RGB
    Alpha=Power(Base=@n1, Exp=@n2)
   )
   B=Add(A=Multiply(A=@n3, B=@n4) {constB=0.75}, B=OneMinus(None=@n4)) {constB=0.25}
  )
  n6 = DynamicParameter() {defaultValue=<rgba 0.5 0.2 0.5 1>, paramNames [Noise Scale, Noise Offset, Bright Spot Intensity, Bright Spot Gradient]}
  n7 = Multiply(A=@n3, B=@n6.Bright Spot Intensity)
  MaterialFunctionCall(Emissive=LinearInterpolate(A=Add(A=@n5, B=@n7), B=VectorParameter() {parameterName=Color Outline, defaultValue=<rgba 0 0 0 1>}.RGB, Alpha=NamedRerouteUsage())) {materialFunction=&MF_FX_EmissiveAdjustments}
  Multiply(A=NamedRerouteUsage(), B=ParticleColor().A)
  n8 = TextureCoordinate()
  n9 = ScalarParameter() {defaultValue=10, parameterName=Rows, sliderMax=30, sliderMin=1}
  n10 = Floor(None=Multiply(A=ComponentMask(None=@n8) {r=true}, B=@n9))
  n11 = ScalarParameter() {defaultValue=5, parameterName=Columns, sliderMax=30, sliderMin=1}
  n12 = Multiply(A=ComponentMask(None=@n8) {g=true}, B=@n11)
  n13 = Divide(A=Floor(None=@n12), B=@n11)
  NamedRerouteDeclaration(None=AppendVector(A=Divide(A=@n10, B=@n9), B=@n13)) {name=Quant UV, nodeColor=<rgba 0.9453 1 0 1>}
  n14 = Frac(None=Multiply(A=ComponentMask(None=@n8) {r=true}, B=@n9))
  n15 = Frac(None=@n12)
  n16 = Add(A=@n8, B=AppendVector(A=DynamicParameter() {parameterIndex=1, paramNames [ScrollSpeed,Param2,Param3,Param4]}.ScrollSpeed, B=Constant()))
  n17 = Multiply(A=ComponentMask(None=Frac(None=@n16)) {r=true}, B=@n9)
  n18 = Frac(None=@n17)
  n19 = Frac(None=@n12)
  NamedRerouteDeclaration(None=LinearInterpolate(A=AppendVector(A=@n14, B=@n15), B=AppendVector(A=@n18, B=@n19), Alpha=ScalarParameter() {defaultValue=1, parameterName=Is Animated})) {name=Frac UV, nodeColor=<rgba 0.7344 0 1 1>}
  n20 = NamedRerouteUsage()
  n21 = Frac(None=Multiply(A=ComponentMask(None=@n20) {g=true}, B=ScalarParameter() {defaultValue=1, parameterName=Noise Texture Tile}) {constB=2})
  n22 = Multiply(A=Time(), B=ScalarParameter() {defaultValue=1, parameterName=Noise Speed, sliderMax=1}) {constB=0.15}
  n23 = TextureSampleParameter2D!(UVs=AppendVector(A=Add(A=@n21, B=@n22) {constB=3}, B=Constant() {desc=a, r=0.5}), Bias, Apply View MipBias) {
   constMipValue=-4
   group=LED Spectrum Bar Graph
   mipValueMode=TMVM_MipBias
   parameterName=Noise Texture
   samplerSource=SSM_Wrap_WorldGroupSettings
   samplerType=SAMPLERTYPE_LinearColor
   texture=/Game/Global/Textures/Default/Caustics/03_WaterFoam_AccentSwirls_H
  }
  n24 = DynamicParameter() {defaultValue=<rgba 0.5 0.2 0.5 1>, paramNames [Noise Scale, Noise Offset, Bright Spot Intensity, Bright Spot Gradient]}
  NamedRerouteDeclaration(None=Add(A=Multiply(A=@n23.R, B=@n24.Noise Scale) {constB=0.5}, B=@n24.Noise Offset) {constB=0.2}) {name=Noise Value, nodeColor=<rgba 0.0021 1 0.1023 1>}
  n25 = ComponentMask(None=NamedRerouteUsage()) {r=true}
  n26 = NamedRerouteUsage()
  n27 = Constant()
  n28 = Constant() {r=1}
  NamedRerouteDeclaration(None=If(A=@n25, B=@n26, "A > B"=@n27, "A < B"=@n28)) {name=Visualizer Mask, nodeColor=<rgba 0.8984 1 0 1>}
  n29 = NamedRerouteUsage()
  n30 = ComponentMask(None=@n29) {r=true}
  n31 = ScalarParameter() {defaultValue=0.05, parameterName=Outline Thickness, sliderMax=1}
  n32 = Constant()
  n33 = Constant() {r=1}
  n34 = Divide(A=ScalarParameter() {defaultValue=0.05, parameterName=Outline Thickness, sliderMax=1})
  n35 = Frac(None=ComponentMask(None=Add(A=@n29, B=@n34)) {g=true})
  n36 = Constant()
  n37 = Constant() {r=1}
  NamedRerouteDeclaration(None=Max(A=If(A=@n30, B=@n31, "A > B"=@n32, "A < B"=@n33), B=If(A=@n35, B=@n31, "A > B"=@n36, "A < B"=@n37))) {name=Outline Mask, nodeColor=<rgba 1 0.0938 0 1>}
  n38 = Divide(B=ScalarParameter() {defaultValue=10, parameterName=Rows, sliderMax=30, sliderMin=1})
  NamedRerouteDeclaration(None=If(A=@n25, B=Subtract(A=@n26, B=@n38), "A > B"=@n27, "A < B"=@n28)) {name=Visualizer Mask_One_Less, nodeColor=<rgba 1 0.3709 0 1>}
  n39 = OneMinus(None=ComponentMask(None=NamedRerouteUsage()) {r=true})
  n40 = NamedRerouteUsage()
  NamedRerouteDeclaration(None=Add(A=Multiply(A=@n39, B=@n40), B=NamedRerouteUsage())) {name=Faded_Mask, nodeColor=<rgba 1 0.0234 0 1>}
  n41 = NamedRerouteUsage()
  n42 = NamedRerouteUsage()
  n43 = ScalarParameter() {defaultValue=1, parameterName=Is Animated}
  n44 = Time()
  n45 = ScalarParameter() {defaultValue=0.05, parameterName=Bright Spot Speed, sliderMax=1}
  n46 = TextureSampleParameter2D(UVs=Add(A=LinearInterpolate(A=@n41, B=@n42, Alpha=@n43), B=Multiply(A=@n44, B=@n45) {constB=0.15})) {
   constMipValue=-4
   group=LED Spectrum Bar Graph
   mipValueMode=TMVM_MipBias
   parameterName=Noise Texture
   samplerSource=SSM_Wrap_WorldGroupSettings
   samplerType=SAMPLERTYPE_LinearColor
   texture=&03_WaterFoam_AccentSwirls_H
  }
  n47 = ComponentMask(None=TextureCoordinate()) {r=true}
  n48 = DynamicParameter() {defaultValue=<rgba 0.5 0.2 0.5 1>, paramNames [Noise Scale, Noise Offset, Bright Spot Intensity, Bright Spot Gradient]}
  NamedRerouteDeclaration(None=Multiply(A=@n46.R, B=Power(Base=@n47, Exp=@n48.Bright Spot Gradient) {constExponent=1})) {name=Bright Spot Noise, nodeColor=<rgba 1 0.4616 0.7489 1>}
  n49 = Floor(None=@n17)
  NamedRerouteDeclaration(None=AppendVector(A=Divide(A=@n49, B=@n9), B=@n13)) {name=QuantUVAnimated, nodeColor=<rgba 0.9688 1 0 1>}
 ]
}
