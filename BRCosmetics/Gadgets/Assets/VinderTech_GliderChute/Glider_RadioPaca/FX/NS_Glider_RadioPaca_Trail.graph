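# Niagara trail effect for the RadioPaca glider cosmetic, combining socket-driven ribbon trails, glowing strips, glitchy square sprites and speed lines.
# Visualizer trails and squares scale with an audio-reactivity value, a system timeline and smoothed interpolations, with lateral forces bending them.
# Glow strips stretch partitioned ribbon particles between socket pairs.
# Variant selectors swap blue and slime-green palettes, and front-end modifiers retune the look outside gameplay.
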NiagaraSystem /BRCosmetics/Gadgets/Assets/VinderTech_GliderChute/Glider_RadioPaca/FX/NS_Glider_RadioPaca_Trail
dependencies {
 usedDataInterfaces [
  &NiagaraDataInterfaceVector2DCurve
  &NiagaraDataInterfaceCurve
  &NiagaraDataInterfaceColorCurve
  &NiagaraDataInterfaceArrayColor
  &NiagaraDataInterfaceSocketReader
 ]
 usedDynamicInputs [
  &RandomRangeVector2D
  &SetBoolByFloatComparison
  &Abs_Float
  &NormalizeVector
  &TransformVector
  &ModuloFloat
 ]
 usedModules [
  &self:SystemSpawnScript.SystemScriptSource.SystemScriptGraph.NiagaraNodeAssignment_1.SetVariables_#1
  &SystemState
  &Timeline
  &FN_Integrator
  &InterpolateOverTime
  &FN_VariantManager
  &self:SystemSpawnScript.SystemScriptSource.SystemScriptGraph.NiagaraNodeAssignment_2.SetVariables_#2
 ]
}
emitters {
 Glitchy_Square {
  emitterUpdateScript [
   &EmitterState(Life Cycle Mode=::NewEnumerator1, Loop Behavior=::NewEnumerator1, Loop Delay=2.4, Loop Duration Mode=::NewEnumerator1, UseLoopDelay=true)
   &SpawnRate(
    SpawnRate=&FN_FrontEndFloatModifier(
     Gameplay=&Multiply_Float(
      A=&Lerp_Float(
       A=&Lerp_Float(A=40, B=80, Alpha=$System.InterpolateOverTime002.InterpolateOverTime_Float.Moving Average)
       B=70
       Alpha=$System.Timeline.NormalizedTime
      )
      B=$System.Audio_Reactivity
     )
     Front End=&Multiply_Float(A=100, B=$System.Audio_Reactivity)
    )
    Use Spawn Probability=false
   )
  ]
  particleSpawnScript [
   &FN_SetSpawnGroup(Number of Spawn Groups=2)
   &InitializeParticle(
    Color=&VariantColorSelector(
     Variant Amount=0
     Original Color=&SelectLinearColorFromArray(
      Array Sampling Mode=::NewEnumerator0
      Direct Array Index=?
      Color Selection Array=DataInterface() {propertyValues {ColorData [<rgba 0 1.618 2 1>, <rgba 0 0.7565 1.5 1>, <rgba 0 0.5208 1.25 1>]}}
      Random Seed=0
     )
     Variant 1 Color=&SelectLinearColorFromArray(
      Array Sampling Mode=::NewEnumerator0
      Direct Array Index=?
      Color Selection Array=DataInterface() {propertyValues {ColorData [<rgba 0.75 1.5 0 1>, <rgba 1 1.5 0 1>, <rgba 0.8417 1 0.05 1>]}}
      Random Seed=0
     )
     Variant 2 Color=?
    )
    Lifetime=0.6
    Sprite Size Mode=::NewEnumerator3
    Uniform Sprite Size=&Multiply_Float(A=&RandomRangeFloat(Minimum=3, Maximum=6), B=1)
    UsePositionOffset=true
    Write Material Random=true
   )
   &SocketLocation(
    SamplePosOffset=true
    Sampled Position Offset=&MakeVector(
     X=&RandomRangeFloat(Minimum=0, Maximum=-30)
     Y=0
     Z=&MaskFloatBySpawnGroup(Masked Float=5, Passthrough Float=-5)
    )
    Socket Index=&MaskIntBySpawnGroup(Masked Int=0, Passthrough Int=2, Spawn Group Mask=0)
    Socket Reader=$System.Socket Reader
    Socket Sampling Mode=::NewEnumerator2
    VelocitySpeedLimit=false
   )
   &AddVelocity(
    Cone Angle=&FN_FrontEndFloatModifier(Gameplay=&Lerp_Float(A=15, B=20, Alpha=$System.Timeline.NormalizedTime), Front End=20)
    Module.UseConeFalloff=false
    Module.UseVelDistribution=false
    Override Seed=false
    Velocity Mode=::NewEnumerator2
    Velocity Speed=&FN_FrontEndFloatModifier(
     Gameplay=&Lerp_Float(
      A=0
      B=&Multiply_Float(
       A=&Multiply_Float(A=&RandomRangeFloat(Minimum=300, Maximum=500), B=$System.Audio_Reactivity)
       B=&Lerp_Float(A=0.5, B=0.8, Alpha=$System.Timeline.NormalizedTime)
      )
      Alpha=&FloatFromCurve(
       FloatCurve=DataInterface() {propertyValues {Curve {
          defaultValue=1
          keys [{interpMode=RCIM_Cubic, tangentMode=RCTM_User, time=1, value=0}, {interpMode=RCIM_Cubic, tangentMode=RCTM_User, time=2, value=1}]
         }}}
       CurveIndex=$Emitter.Age
      )
     )
     Front End=&Multiply_Float(
      A=&RandomRangeFloat(Minimum=125, Maximum=200)
      B=&Add_Float(A=1, B=&Multiply_Float(A=$System.Audio_Reactivity, B=0.75))
     )
    )
   )
   &CameraOffset(Camera Offset Amount=5)
  ]
  particleUpdateScript [
   &ParticleState()
   &LinearForce(Force=&Multiply_VectorByFloat(
     Vector=&Lerp_Vector(
      A=<0 0 0>
      B=<0 1 0>
      Alpha=&Multiply_Float(A=$System.Timeline001.Time, B=$System.InterpolateOverTime001.InterpolateOverTime_Float.Moving Average)
     )
     Float=&FloatFromCurve(
      FloatCurve=DataInterface() {propertyValues {Curve {
         defaultValue=1
         keys [
          {arriveTangent=-0.0013, interpMode=RCIM_Cubic, leaveTangent=-0.0013, tangentMode=RCTM_User, time=0, value=0}
          {arriveTangent=1.3901, interpMode=RCIM_Cubic, leaveTangent=1.3901, tangentMode=RCTM_User, time=1, value=1}
         ]
        }}}
      Scale Curve=&Multiply_Float(A=300, B=&Lerp_Float(A=1, B=1.8, Alpha=$System.Timeline.NormalizedTime))
     )
    ))
   &Drag(Drag=3, Use Linear Drag=true, Use Rotational Drag=false)
   &ScaleSpriteSize(Scale Factor=&Vector2DFromFloat(Value=&FloatFromCurve(FloatCurve=DataInterface() {propertyValues {Curve {
         defaultValue=1
         keys [{interpMode=RCIM_Cubic, tangentMode=RCTM_User, time=0, value=0.5}, {interpMode=RCIM_Cubic, tangentMode=RCTM_User, time=0.7, value=1}]
        }}}))) {module=ScaleSpriteSize001}
   &ScaleColor(
    Scale Alpha=&FloatFromCurve(FloatCurve=DataInterface() {propertyValues {Curve {
        defaultValue=0
        keys [
         {interpMode=RCIM_Cubic, tangentMode=RCTM_User, time=0, value=0}
         {interpMode=RCIM_Cubic, tangentMode=RCTM_User, time=0.2, value=1}
         {tangentMode=RCTM_User, time=1, value=0}
        ]
       }}})
    ScaleA=true
    ScaleRGB=true
   ) {module=ScaleColor001}
   &CurlNoiseForce(
    Bias Noise=false
    Noise Strength=&Multiply_Float(A=500, B=&Lerp_Float(A=0.3, B=1, Alpha=$System.Timeline.NormalizedTime))
    Pan Noise=false
   )
   &SolveForcesAndVelocity(Clamp Velocity=false, Limit Acceleration=false)
   &SpriteFacingAndAlignment(Sprite Facing=$Particles.ShapeLocation.ShapeTangent, WriteAlignment=false, WriteFacing=true)
   &FN_Sprite_Attributes(
    Color=&MultiplyLinearColor(Linear Color=$Particles.Color, Scale Factor=<1 1 1 1>)
    Create Color Attribute=true
    Create SpriteSize Attribute=true
    SpriteSize=&Multiply_Vector2D_ByFloat(Float=1.5, Vector2D=$Particles.SpriteSize)
    Write to Module Namespace=true
   )
  ]
  rendererClasses [&NiagaraSpriteRendererProperties]
  renderers [
   {rendererClass=&NiagaraSpriteRendererProperties}
   {rendererClass=&NiagaraSpriteRendererProperties, rendererIndex=1}
  ]
 }
 Glow_Strip {
  emitterUpdateScript [
   &EmitterState(Life Cycle Mode=::NewEnumerator1, Loop Behavior=::NewEnumerator1, Loop Duration Mode=::NewEnumerator1, UseLoopDelay=false)
   &SpawnBurst_Instantaneous(Spawn Count=10, Use Spawn Probability=false, UseLoopCountLimit=false)
  ]
  particleSpawnScript [
   &InitializeParticle(
    Color=&VariantColorSelector(Variant Amount=0, Original Color=<rgba 0 0.7565 1.5 1>, Variant 1 Color=<rgba 1 2 0 1>, Variant 2 Color=?)
    Lifetime=3
    Position Mode=::NewEnumerator0
    Ribbon Width=3
    Ribbon Width Mode=::NewEnumerator1
    Write Material Random=true
   )
   &PartitionParticles(
    Normalized Index In Partition Includes One=true
    Particles Per Partition=5
    Partition Mode:ENiagara_PartitionMode_v2=::NewEnumerator1
    Write Partition Index to Ribbon Link Order:ENiagara_AttributeSamplingApplyOutput=::NewEnumerator0
    Write Partition to Ribbon ID=::NewEnumerator0
   )
  ]
  particleUpdateScript [
   &ParticleState(Kill Particles When Lifetime Has Elapsed=false)
   &SolveForcesAndVelocity(Clamp Velocity=false, Limit Acceleration=false)
   &SocketLocation(
    Position Sampling=::NewEnumerator1
    SamplePosOffset=true
    Sampled Position Offset=<1.9 0 -1>
    Socket Reader=$System.Socket Reader
    Socket Sampling Mode=::NewEnumerator3
    VelocitySpeedLimit=false
   )
   &SocketLocation(
    Position Sampling=::NewEnumerator1
    SamplePosOffset=true
    Sampled Position Offset=<-1.8 0 0>
    Socket Index=1
    Socket Reader=$System.Socket Reader
    Socket Sampling Mode=::NewEnumerator3
    VelocitySpeedLimit=false
   ) {module=SocketLocation001}
   &SocketLocation(
    Position Sampling=::NewEnumerator1
    SamplePosOffset=true
    Sampled Position Offset=<1.9 0 1>
    Socket Index=2
    Socket Reader=$System.Socket Reader
    Socket Sampling Mode=::NewEnumerator3
    VelocitySpeedLimit=false
   ) {module=SocketLocation002}
   &SocketLocation(
    Position Sampling=::NewEnumerator1
    SamplePosOffset=true
    Sampled Position Offset=<-1.8 0 0>
    Socket Index=3
    Socket Reader=$System.Socket Reader
    Socket Sampling Mode=::NewEnumerator3
    VelocitySpeedLimit=false
   ) {module=SocketLocation003}
   /SetVariables(Particles.Position=&AddVectorToPosition(
     Position=&MergeHelper_VectorSwitch(
      0=&Lerp_Vector(
       A=Linked() {linkedVariable {name=Output.SocketLocation.SampledPosition, type=NiagaraPosition}}
       B=Linked() {linkedVariable {name=Output.SocketLocation001.SampledPosition, type=NiagaraPosition}}
       Alpha=$Particles.PartitionParticles.Normalized Index In Partition
      )
      1=&Lerp_Vector(
       A=Linked() {linkedVariable {name=Output.SocketLocation002.SampledPosition, type=NiagaraPosition}}
       B=Linked() {linkedVariable {name=Output.SocketLocation003.SampledPosition, type=NiagaraPosition}}
       Alpha=$Particles.PartitionParticles.Normalized Index In Partition
      )
      2=?
      3=?
      Index=$Particles.PartitionParticles.Partition
      Num=::NewEnumerator0
     )
     Vector=<-0.2 0 0>
    )) {isSetParameters=true}
   &ScaleColor(
    Scale Alpha=&FloatFromCurve(
     FloatCurve=DataInterface() {propertyValues {Curve {
        defaultValue=1
        keys [{interpMode=RCIM_Cubic, tangentMode=RCTM_User, time=2.5, value=0}, {interpMode=RCIM_Cubic, tangentMode=RCTM_User, time=3.5, value=1}]
       }}}
     CurveIndex=$Emitter.Age
     Scale Curve=5
    )
    Scale RGB=&VectorFromFloat(Value=5)
    ScaleA=true
    ScaleRGB=true
   )
   &ScaleRibbonWidth(Ribbon Width Scale=&FloatFromCurve(
     FloatCurve=DataInterface() {propertyValues {Curve {keys [<at 0 1>,<at 1 0.75>]}}}
     CurveIndex=$Particles.PartitionParticles.Normalized Index In Partition
    ))
  ]
  rendererClasses [&NiagaraRibbonRendererProperties]
  renderers [{rendererClass=&NiagaraRibbonRendererProperties}]
 }
 Speed_Lines {
  emitterUpdateScript [
   &EmitterState()
   &SpawnRate(
    SpawnRate=&FN_FrontEndFloatModifier(Gameplay=&Lerp_Float(A=0, B=90, Alpha=$System.Timeline.NormalizedTime), Front End=50)
    Use Spawn Probability=true
   ) {module=SpawnRate001}
  ]
  particleSpawnScript [
   &InitializeParticle(
    Color=&MakeLinearColorFromVectorAndFloat(
     "Vector (RGB)"=<1 1 1>
     "Float (Alpha)"=&Multiply_Float(
      A=&UniformRangedFloat(Minimum=0.1, Maximum=0.15)
      B=&FN_FrontEndFloatModifier(Gameplay=0.5, Front End=0.5)
     )
    )
    Lifetime Max=0.16
    Lifetime Min=0.1
    Lifetime Mode=::NewEnumerator1
    Sprite Size Max=<3 10>
    Sprite Size Min=<2 2>
    Sprite Size Mode=::NewEnumerator2
    UsePositionOffset=false
    Write Material Random=true
   )
   &TorusLocation(
    Align Torus With Axis=true
    Handle Radius=0
    Large Radius:NiagaraFloat=&FN_FrontEndFloatModifier(
     Gameplay=&RandomRangeFloat(Minimum=150, Maximum=200)
     Front End=&RandomRangeFloat(Minimum=200, Maximum=270)
    )
    MaskSpawnGroup=false
    Module.Override Randomness=false
    Torus Axis=<1 0 0>
    Torus Origin:NiagaraPosition=&AddVectorToPosition(
     Position=$Particles.Position
     Vector=&FN_FrontEndVector3Modifier(Gameplay=<0 0 0>, Front End=<-200 0 0>)
    )
   )
   &AddVelocityFromPoint(
    Module.Use Falloff Distance=false
    Origin Offset=<-500 0 0>
    Velocity Strength=&Multiply_Float(
     A=&FN_FrontEndFloatModifier(Gameplay=600, Front End=0)
     B=&MakeCustomFloatFromBool(Bool=$User.bIsFrontEndPreview, True Float=-1, False Float=1)
    )
   )
   &AddVelocity(Velocity=&Multiply_Vector(
     A=&RandomRangeVector(Minimum=<300 0 0>, Maximum=<700 0 0>)
     B=&MakeVector(X=1, Y=&MakeCustomFloatFromBool(Bool=$User.bIsFrontEndPreview, True Float=-1, False Float=1), Z=1)
    ))
  ]
  particleUpdateScript [
   &SolveForcesAndVelocity(Clamp Velocity=false, Limit Acceleration=false)
   &ParticleState()
   &ScaleSpriteSize(Scale Factor=&Vector2DFromCurve(
     Vector2Curve=DataInterface() {propertyValues {XCurve {keys [<at 0 0.5 RCIM_Cubic>,<at 0.5 1 RCIM_Cubic>]}, YCurve {keys [<at 0 0 RCIM_Cubic>,<at 0.5 1 RCIM_Cubic>]}}}
     ScaleCurve=&Vector2DFromFloat(Value=1)
    ))
   &ScaleSpriteSizeBySpeed(Max Scale Factor=<1 50>, Min Scale Factor=<1 40>, Velocity Threshold=100)
   &ScaleColor(
    Scale Alpha=&FN_FrontEndFloatModifier(
     Gameplay=&Lerp_Float(A=0, B=1, Alpha=$System.Timeline.NormalizedTime)
     Front End=&FloatFromCurve(
      FloatCurve=DataInterface() {propertyValues {Curve {
         defaultValue=1
         keys [{interpMode=RCIM_Cubic, tangentMode=RCTM_User, time=2.3, value=0}, {interpMode=RCIM_Cubic, tangentMode=RCTM_User, time=2.8, value=1}]
        }}}
      CurveIndex=$Emitter.Age
      Scale Curve=0.5
     )
    )
    Scale RGB=&VectorFromFloat(Value=1)
    ScaleA=true
    ScaleRGB=true
   )
  ]
  rendererClasses [&NiagaraSpriteRendererProperties]
  renderers [{rendererClass=&NiagaraSpriteRendererProperties}]
 }
 Visualizer_Trails {
  emitterSpawnScript [&FN_RibbonIDOffsetUtility()]
  emitterUpdateScript [
   &EmitterState(Life Cycle Mode=::NewEnumerator1, Loop Behavior=::NewEnumerator1, Loop Duration Mode=::NewEnumerator1, UseLoopDelay=false)
   &FN_SpawnAnimTrail(Number of Anim Trails=2)
  ]
  particleSpawnScript [
   &FN_InitializeAnimTrailBoneIndecies(First Bone or Socket Index=1, Second Bone or Socket Index=0)
   &FN_InitializeAnimTrailBoneIndecies(AnimTrail Selection=1, First Bone or Socket Index=3, Second Bone or Socket Index=2) {module=FN_InitializeAnimTrailBoneIndecies001}
   &FN_InitializeAnimTrail(
    Color=&VariantColorSelector(Variant Amount=0, Original Color=<rgba 0 1.5 2 1>, Variant 1 Color=<rgba 1.1667 2 0 1>, Variant 2 Color=?)
    Lifetime=&MakeCustomFloatFromBool(Bool=$Particles.Locked, True Float=0, False Float=0.65)
   )
   &AddVelocity(Velocity=&Multiply_VectorByFloat(
     Vector=&Add_Vector(
      A=<3 0 0>
      B=&Lerp_Vector(
       A=<0 0 0>
       B=<0 0 0>
       Alpha=&Multiply_Float(A=$System.Timeline001.Time, B=$System.InterpolateOverTime001.InterpolateOverTime_Float.Moving Average)
      )
     )
     Float=&FN_FrontEndFloatModifier(
      Gameplay=&Lerp_Float(
       A=0
       B=60
       Alpha=&FloatFromCurve(
        FloatCurve=DataInterface() {propertyValues {Curve {
           defaultValue=1
           keys [{interpMode=RCIM_Cubic, tangentMode=RCTM_User, time=1, value=0}, {interpMode=RCIM_Cubic, tangentMode=RCTM_User, time=2, value=1}]
          }}}
        CurveIndex=$Emitter.Age
       )
      )
      Front End=60
     )
    ))
  ]
  particleUpdateScript [
   &LinearForce(Force=&Multiply_VectorByFloat(
     Vector=&Lerp_Vector(
      A=<0 0 0>
      B=<0 1 0>
      Alpha=&Multiply_Float(A=$System.Timeline001.Time, B=$System.InterpolateOverTime001.InterpolateOverTime_Float.Moving Average)
     )
     Float=&FloatFromCurve(
      FloatCurve=DataInterface() {propertyValues {Curve {
         defaultValue=1
         keys [
          {arriveTangent=0.4626, interpMode=RCIM_Cubic, leaveTangent=0.4626, tangentMode=RCTM_User, time=0, value=0}
          {arriveTangent=1.7891, interpMode=RCIM_Cubic, leaveTangent=1.7891, tangentMode=RCTM_User, time=1, value=1}
         ]
        }}}
      Scale Curve=1000
     )
    ))
   &SolveForcesAndVelocity(Clamp Velocity=false, Limit Acceleration=false)
   &FN_UpdateAnimTrail(Locked Particle RibbonWidth=1) {module=FN_UpdateAnimTrail001}
   &ScaleColor(
    Scale Alpha=&Multiply_Float(
     A=&FN_FrontEndFloatModifier(
      Gameplay=&Lerp_Float(
       A=&Lerp_Float(A=0.5, B=0.6, Alpha=$System.InterpolateOverTime002.InterpolateOverTime_Float.Moving Average)
       B=1
       Alpha=$System.Timeline.NormalizedTime
      )
      Front End=0.35
     )
     B=&FloatFromCurve(
      FloatCurve=DataInterface() {propertyValues {Curve {
         defaultValue=1
         keys [{interpMode=RCIM_Cubic, tangentMode=RCTM_User, time=0, value=1}, {interpMode=RCIM_Cubic, tangentMode=RCTM_User, time=0.75, value=0}]
        }}}
      Scale Curve=&FloatFromCurve(
       FloatCurve=DataInterface() {propertyValues {Curve {
          defaultValue=1
          keys [{interpMode=RCIM_Cubic, tangentMode=RCTM_User, time=2.3, value=0}, {interpMode=RCIM_Cubic, tangentMode=RCTM_User, time=2.8, value=1}]
         }}}
       CurveIndex=$Emitter.Age
      )
     )
    )
    Scale RGB=&MakeVectorFromLinearColorRGB(Color=&FNTimeOfDayColor(
      Night Color=<rgba 0.5625 0.6225 0.75 1>
      Morning Color=<rgba 0.7812 0.5615 0.5615 1>
      Day Color=<rgba 1 1 1 1>
      Evening Color=<rgba 0.7812 0.5615 0.5615 1>
     ))
    ScaleA=true
    ScaleRGB=true
   )
   /SetVariables(Particles.DynForSoftTrail:Vector4f=&MakeVector4(
     X=1
     Y=0.75
     Z=&FN_FrontEndFloatModifier(
      Gameplay=&Multiply_Float(
       A=&Lerp_Float(
        A=&Lerp_Float(A=0.65, B=0.95, Alpha=$System.InterpolateOverTime002.InterpolateOverTime_Float.Moving Average)
        B=0.8
        Alpha=$System.Timeline.NormalizedTime
       )
       B=$System.Audio_Reactivity
      )
      Front End=&Add_Float(A=0.2, B=$System.Audio_Reactivity)
     )
     W=0
    )) {isSetParameters=true}
   /SetVariables(Particles.RibbonWidth02:NiagaraFloat=&Multiply_Float(A=$Particles.RibbonWidth, B=1.1)) {isSetParameters=true}
   &ScaleRibbonWidth(Ribbon Width Scale=&FloatFromCurve(FloatCurve=DataInterface() {propertyValues {Curve {defaultValue=0, keys [<at 0 1>,<at 1 1.2>]}}}))
   &DynamicMaterialParameters(
    Index 0 Param 1=&FN_FrontEndFloatModifier(
     Gameplay=&Multiply_Float(
      A=&Lerp_Float(
       A=&Lerp_Float(A=0.2, B=0.25, Alpha=$System.InterpolateOverTime002.InterpolateOverTime_Float.Moving Average)
       B=0.3
       Alpha=$System.Timeline.NormalizedTime
      )
      B=$System.Audio_Reactivity
     )
     Front End=&Multiply_Float(A=0.3, B=$System.Audio_Reactivity)
    )
    Index 0 Param 2=&FN_FrontEndFloatModifier(
     Gameplay=&Multiply_Float(
      A=&Add_Float(
       A=&FloatFromCurve(
        FloatCurve=DataInterface() {propertyValues {Curve {
           defaultValue=1
           keys [{interpMode=RCIM_Cubic, tangentMode=RCTM_User, time=2, value=-1}, {interpMode=RCIM_Cubic, tangentMode=RCTM_User, time=3, value=0}]
          }}}
        CurveIndex=$Emitter.Age
       )
       B=&Lerp_Float(
        A=&Lerp_Float(A=0.6, B=0.9, Alpha=$System.InterpolateOverTime002.InterpolateOverTime_Float.Moving Average)
        B=0.75
        Alpha=$System.Timeline.NormalizedTime
       )
      )
      B=$System.Audio_Reactivity
     )
     Front End=&Add_Float(
      A=&FloatFromCurve(
       FloatCurve=DataInterface() {propertyValues {Curve {
          defaultValue=1
          keys [{interpMode=RCIM_Cubic, tangentMode=RCTM_User, time=2, value=-1}, {interpMode=RCIM_Cubic, tangentMode=RCTM_User, time=3, value=0}]
         }}}
       CurveIndex=$Emitter.Age
      )
      B=$System.Audio_Reactivity
     )
    )
    Index 0 Param 3=&FN_FrontEndFloatModifier(Gameplay=&Lerp_Float(A=0, B=0.15, Alpha=$System.Timeline.NormalizedTime), Front End=0.2)
    Index 0 Param 4=&FN_FrontEndFloatModifier(Gameplay=&Lerp_Float(A=0.3, B=0.15, Alpha=$System.Timeline.NormalizedTime), Front End=0.15)
    Module.Index0Float0Write=true
    Module.Index0Float1Write=true
    Module.Index0Float2Write=true
    Module.Index0Float3Write=true
   )
   /SetVariables(
    Particles.BlueVariantColor:LinearColor=&MakeLinearColorFromVectorAndFloat(
     "Vector (RGB)"=&MakeVectorFromLinearColorRGB(Color=<rgba 1 1 1 1>)
     "Float (Alpha)"=/BRCosmetics/Weapons/FORT_Melee/Pickaxe_GateHound/FX/VariantManager/VariantFloatSelector(
      Original Float:NiagaraFloat=1
      Variant 1 Float:NiagaraFloat=0
      Variant 2 Float:NiagaraFloat=?
      Variant 3 Float:NiagaraFloat=?
      Variant Amount:NiagaraInt32=0
     )
    )
    Particles.SlimeVariantColor:LinearColor=&MakeLinearColorFromVectorAndFloat(
     "Vector (RGB)"=&MakeVectorFromLinearColorRGB(Color=<rgba 1 1 1 1>)
     "Float (Alpha)"=&VariantFloatSelector(
      Original Float:NiagaraFloat=0
      Variant 1 Float:NiagaraFloat=1
      Variant 2 Float:NiagaraFloat=?
      Variant 3 Float:NiagaraFloat=?
      Variant Amount:NiagaraInt32=0
     )
    )
   ) {isSetParameters=true}
  ]
  rendererClasses [&NiagaraRibbonRendererProperties]
  renderers [
   {rendererClass=&NiagaraRibbonRendererProperties}
   {rendererClass=&NiagaraRibbonRendererProperties, rendererIndex=1}
   {rendererClass=&NiagaraRibbonRendererProperties, rendererIndex=2}
  ]
 }
}
summary {
 systemName=NS_Glider_RadioPaca_Trail
 emitters [
  {emitterName=Speed_Lines, rendererClasses [&NiagaraSpriteRendererProperties]}
  {emitterName=Visualizer_Trails, rendererClasses [&NiagaraRibbonRendererProperties]}
  {emitterName=Glitchy_Square, rendererClasses [&NiagaraSpriteRendererProperties]}
  {emitterName=Glow_Strip, rendererClasses [&NiagaraRibbonRendererProperties]}
 ]
 userVariables [
  User.bIsFrontEnd {type=NiagaraBool}
  User.bIsFrontEndPreview {type=NiagaraBool}
  User.ForwardDot {type=NiagaraFloat, defaultValue=0, description=""}
  User.RibbonIDOffset {type=NiagaraInt32}
  User.RightDot {type=NiagaraFloat, defaultValue=0, description=""}
  User.Variant {type=NiagaraFloat, defaultValue=0, description=""}
 ]
}
